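# Physics asset for the M_MED_Grasshopper backpack cosmetic, defining collision bodies and joints for its dangling accessory parts.
# Capsules cover the pack and horn, while boxes cover the buckle, pouch strap and master pouch.
# Constraints link the horn to the pack and chain the strap through the buckle to the pack, with mostly tight swing limit values and the widest twist on the strap.
# The master pouch body has no constraint.
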PhysicsAsset /BRCosmetics/Accessories/FORT_Backpacks/Backpack_M_MED_Grasshopper/Meshes/M_MED_Grasshopper_Pack_Physics
bodies {
 dyn_buckle_jnt [{shapeType=Box, extentX=2, extentY=1, extentZ=2, shapeName=dyn_buckle_jnt_box, center=<0 0.0709 -1.2255>}]
 dyn_horn_jnt [{shapeType=Capsule, length=25.25, radius=4.5, shapeName=dyn_horn_jnt_capsule, center=<3.1357 0.0001 0>}]
 dyn_pouch_strap_jnt1 [{shapeType=Box, extentX=2, extentY=1, extentZ=2, shapeName=dyn_pouch_strap_jnt1_box, center=<-0.0001 -0.7207 -1.1834>}]
 master_pouch_jnt [{shapeType=Box, extentX=3.5, extentY=7, extentZ=12, shapeName=master_pouch_jnt_box, center=<0.1419 0 0>}]
 pack [{shapeType=Capsule, length=16, radius=7, shapeName=pack_capsule, center=<0 -4.9102 -9>}]
}
constraints [
 {bone1Name=dyn_horn_jnt, bone2Name=pack, swing1LimitDegrees=2, swing2LimitDegrees=0.5, twistLimitDegrees=10}
 {bone1Name=dyn_pouch_strap_jnt1, bone2Name=dyn_buckle_jnt, swing1LimitDegrees=10, swing2LimitDegrees=1, twistLimitDegrees=45}
 {bone1Name=dyn_buckle_jnt, bone2Name=pack, swing1LimitDegrees=10, swing2LimitDegrees=1, twistLimitDegrees=5}
]
